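# Unlit material for a prototype Landmass erosion simulation pass that gathers water and sediment from height and velocity render textures.
# Custom HLSL scans neighboring texels, moves their water by velocity and accumulates bilinear-weighted volume and sediment, with static switches choosing sub-particle variants (two by default).
# Velocity-based solubility drives dissolution or deposition, combined with advected, slope-smoothed height, evaporation and rain to form new height and water values, using MPC_Erosion settings.
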Material /Landmass/Landscape/OldPrototype_BP/Mats/ScatterGather/GatherData
graph {
 properties {
  bAllowNegativeEmissiveColor=true
  blendMode=BLEND_AlphaComposite
  pixelDepthOffsetMode=PDOM_Legacy
  shadingModel=MSM_Unlit
  usedShadingModels=Unlit
 }
 nodes [
  Add()
  n1 = Multiply(
   A=TextureCoordinate()
   B=AppendVector(A=ScalarParameter() {defaultValue=2048, parameterName=SizeX}, B=ScalarParameter() {defaultValue=2048, parameterName=SizeY})
  )
  n2 = VectorNoise(World Position=AppendVector(A=Floor(None=@n1), B=Constant()))
  AppendVector(
   A=Multiply(A=ComponentMask(None=@n2) {g=true, r=true}) {constB=4}
   B=Add(
    A=ScalarParameter() {defaultValue=25, parameterName=Z}
    B=TextureSampleParameter2D() {parameterName=HeightRT, texture=&Black}.R
   )
  )
  MaterialExpressionTextureObjectParameter_1 = TextureObjectParameter() {parameterName=HeightRT, texture=&Black}
  MaterialExpressionTextureObjectParameter_5 = TextureObjectParameter() {parameterName=VelocityRT, texture=&Black}
  MaterialExpressionCollectionParameter_6 = CollectionParameter() {collection=/Landmass/Landscape/BlueprintBrushes/MPC/MPC_Erosion, parameterName=deltaT}
  MaterialExpressionAppendVector_6 = AppendVector(
   A=CollectionParameter() {collection=&MPC_Erosion, parameterName=ResX}
   B=CollectionParameter() {collection=&MPC_Erosion, parameterName=ResY}
  )
  MaterialExpressionDivide_0 = Divide(B=@MaterialExpressionAppendVector_6)
  MaterialExpressionCollectionParameter_2 = CollectionParameter() {collection=&MPC_Erosion, parameterName=GridSize}
  MaterialExpressionMultiply_14 = Multiply(A=@MaterialExpressionAppendVector_6, B=@MaterialExpressionCollectionParameter_2)
  MaterialExpressionMaterialFunctionCall_5 = MaterialFunctionCall() {materialFunction=/Landmass/Landscape/OldPrototype_BP/Mats/MF/PointSampleUVs}
  n3 = Custom!(
   HeightRT=@MaterialExpressionTextureObjectParameter_1
   VelocityRT=@MaterialExpressionTextureObjectParameter_5
   deltaT=@MaterialExpressionCollectionParameter_6
   TextureSize=@MaterialExpressionAppendVector_6
   TexelSize=@MaterialExpressionDivide_0
   TotalSize=@MaterialExpressionMultiply_14
   GridSize=@MaterialExpressionCollectionParameter_2
   inUV=@MaterialExpressionMaterialFunctionCall_5.TexelCenterUVs
  ) {
   code="//PositionRT: input Texture Object containing \r\n//VelcotiyRT: input \r\n//deltaT: input\r\n//input TextureSize\r\n//input TexelSize\r\n//input OutputVelocity\r\n\r\nfloat startheight = HeightRT.SampleLevel(HeightRTSampler, inUV, 0).r;\r\n\r\nint2 startidx = floor(inUV * TextureSize);\r\n\r\nfloat accumvol = 0;\r\nfloat accumZ = 0;\r\nfloat accumsediment = 0;\r\n\r\nfloat accum = 0;\r\nfloat neighbors = 0;\r\n\r\nconst int Radius = 1;\r\n\r\nconst float2 posoffsets[4] = {float2(TexelSize.x,0), float2(-TexelSize.x,0), float2(0,TexelSize.y), float2(0,-TexelSize.y)};\r\n\r\nfor (int j = -Radius; j <= Radius; j++)\r\n{\r\n\tfor (int i = -Radius; i <= Radius; i++)\r\n\t{\r\n\t\t\r\n\t\tfloat2 curUV = inUV + (float2(i,j) * TexelSize);\r\n\t\t//R = height, G = water height, B = water volume, A = sediment fraction\r\n\t\tfloat4 height = HeightRT.SampleLevel(HeightRTSampler, curUV, 0).rgba;\r\n\t\theight.b /= 4;\r\n\r\n\t\tfloat2 centervelocity = VelocityRT.SampleLevel(VelocityRTSampler, curUV, 0).rg;\r\n\t\tfloat2 offsetXvelocity = VelocityRT.SampleLevel(VelocityRTSampler, curUV - float2(TexelSize.x,0), 0).rg;\r\n\t\tfloat2 offsetYvelocity = VelocityRT.SampleLevel(VelocityRTSampler, curUV - float2(0,TexelSize.y), 0).rg;\r\n\t\t\r\n\t\tfloat2 filter1velocity = VelocityRT.SampleLevel(VelocityRTSampler, curUV + float2(+TexelSize.x * 0.5, -TexelSize.y *0.5), 0).rg;\r\n\t\tfloat2 filter2velocity = VelocityRT.SampleLevel(VelocityRTSampler, curUV + float2(-TexelSize.x * 0.5, -TexelSize.y *0.5), 0).rg;\r\n\t\tfloat2 filter3velocity = VelocityRT.SampleLevel(VelocityRTSampler, curUV + float2(-TexelSize.x * 0.5, +TexelSize.y *0.5), 0).rg;\r\n\t\t\r\n\t\tfloat2 velocities[4];\r\n\t\t\r\n\t\tvelocities[0] = float2( centervelocity.x, filter1velocity.y);\r\n\t\tvelocities[1] = float2( offsetXvelocity.x, filter2velocity.y);\r\n\t\tvelocities[2] = float2( filter3velocity.x, centervelocity.y);\r\n\t\tvelocities[3] = float2( filter2velocity.x, offsetYvelocity.y);\r\n\t\tcentervelocity = float2( centervelocity.x + offsetXvelocity.x, centervelocity.y + offsetYvelocity.y) * 0.5; \r\n\t\t\r\n\t\tfor (int k = 0; k < 4; k++)\r\n\t\t{\r\n\t\t\tfloat2 position = lerp(velocities[k],centervelocity,0) * deltaT;\r\n\t\t\r\n\t\t\tfloat2 uvpos = curUV + (position.rg / TotalSize);\r\n\t\t\t//uvpos += posoffsets[k] * 0.5;\r\n\t\r\n\t\t\tfloat2 celldist = abs(inUV - uvpos);\r\n\t\r\n\t\t\tif (celldist.x < TexelSize.x &&&& celldist.y < TexelSize.y &&&& height.b > 0.0001)\r\n\t\t\t{\r\n\t\r\n\t\t\t\tfloat weight = 1 - saturate(celldist.x / TexelSize.x);\r\n\t\t\t\tweight *= 1 - saturate(celldist.y / TexelSize.y);\r\n\t\r\n\t\t\t\t//accumulate Volume, Sediment and Height\r\n\t\t\t\taccumvol += height.b * weight;\r\n\t\t\t\taccumsediment += height.a * height.b * weight;\r\n\t\t\t\t//accumZ += (height.g+ (velocity.z * deltaT)) * height.b * weight;\r\n\t\r\n\t\t\t\taccum += height.b * weight;\r\n\t\t\t}\r\n\t\t\t\r\n\t\t\t}\r\n\t}\r\n}\r\n\r\nif(accum >0)\r\n{\r\n\taccumZ /= accum;\r\n\taccumsediment /= accum;\r\n}\r\nelse\r\n{\r\n\taccumZ=startheight;\r\n\taccumvol=0;\r\n}\r\n\r\nfloat4 output = float4(startheight, max(startheight,startheight), accumvol, accumsediment);\r\n\r\nreturn output;"
   description=Erosion Pass
   outputType=CMOT_Float4
   inputs [
    {
     inputName=HeightRT
     input {expression=&self:MaterialExpressionTextureObjectParameter_1}
    }
    {
     inputName=VelocityRT
     input {expression=&self:MaterialExpressionTextureObjectParameter_5}
    }
    {
     inputName=deltaT
     input {expression=&self:MaterialExpressionCollectionParameter_6}
    }
    {inputName=TextureSize, input {expression=&self:MaterialExpressionAppendVector_6}}
    {inputName=TexelSize, input {expression=&self:MaterialExpressionDivide_0}}
    {inputName=TotalSize, input {expression=&self:MaterialExpressionMultiply_14}}
    {
     inputName=GridSize
     input {expression=&self:MaterialExpressionCollectionParameter_2}
    }
    {
     inputName=inUV
     input {expression=&self:MaterialExpressionMaterialFunctionCall_5}
    }
   ]
  }
  MaterialExpressionTextureObjectParameter_0 = TextureObjectParameter() {parameterName=HeightRT, texture=&Black}
  MaterialExpressionTextureObjectParameter_2 = TextureObjectParameter() {parameterName=VelocityRT, texture=&Black}
  MaterialExpressionCollectionParameter_22 = CollectionParameter() {collection=&MPC_Erosion, parameterName=deltaT}
  MaterialExpressionAppendVector_7 = AppendVector(
   A=CollectionParameter() {collection=&MPC_Erosion, parameterName=ResX}
   B=CollectionParameter() {collection=&MPC_Erosion, parameterName=ResY}
  )
  MaterialExpressionDivide_3 = Divide(B=@MaterialExpressionAppendVector_7)
  MaterialExpressionCollectionParameter_21 = CollectionParameter() {collection=&MPC_Erosion, parameterName=GridSize}
  MaterialExpressionMultiply_0 = Multiply(A=@MaterialExpressionAppendVector_7, B=@MaterialExpressionCollectionParameter_21)
  MaterialExpressionMaterialFunctionCall_4 = MaterialFunctionCall() {materialFunction=&PointSampleUVs}
  n4 = Custom(
   HeightRT=@MaterialExpressionTextureObjectParameter_0
   VelocityRT=@MaterialExpressionTextureObjectParameter_2
   deltaT=@MaterialExpressionCollectionParameter_22
   TextureSize=@MaterialExpressionAppendVector_7
   TexelSize=@MaterialExpressionDivide_3
   TotalSize=@MaterialExpressionMultiply_0
   GridSize=@MaterialExpressionCollectionParameter_21
   inUV=@MaterialExpressionMaterialFunctionCall_4.TexelCenterUVs
  ) {
   code="//PositionRT: input Texture Object containing \r\n//VelcotiyRT: input \r\n//deltaT: input\r\n//input TextureSize\r\n//input TexelSize\r\n//input OutputVelocity\r\n\r\nfloat startheight = HeightRT.SampleLevel(HeightRTSampler, inUV, 0).r;\r\n\r\nint2 startidx = floor(inUV * TextureSize);\r\n\r\nfloat accumvol = 0;\r\nfloat accumZ = 0;\r\nfloat accumsediment = 0;\r\n\r\nfloat accum = 0;\r\nfloat neighbors = 0;\r\n\r\nconst float2 offsets[2] = {float2(0.5 * TexelSize.x,0), float2(0,0.5 * TexelSize.y)};\r\nconst float2 veloffsets[2] = {float2(0.5 * TexelSize.x,-0.5 * TexelSize.y), float2(-0.5 * TexelSize.x,0.5 * TexelSize.y)};\r\n\r\nconst int Radius =2;\r\n\r\nfor (int j = -Radius; j < Radius; j++)\r\n{\r\n\tfor (int i = -Radius; i < Radius; i++)\r\n\t{\r\n\t\t\r\n\t\tfloat2 curUV = inUV + (float2(i,j) * TexelSize);\r\n\r\n\t\t//R = height, G = water height, B = water volume, A = sediment fraction\r\n\t\tfloat4 height = HeightRT.SampleLevel(HeightRTSampler, curUV, 0).rgba;\r\n\r\n\t\t//Offset grid velocity\r\n\t\tfloat3 velocity = VelocityRT.SampleLevel(VelocityRTSampler, curUV, 0).rgb;\r\n\r\n\t\tfor (int k = 0; k < 2; k++)\r\n\t\t{\r\n\r\n\t\t\t//Get water at each of 2 boundaries\r\n\t\t\tfloat4 height = HeightRT.SampleLevel(HeightRTSampler, curUV + offsets[k], 0).rgba;\r\n\t\t\t\r\n\t\t\t//half since we evaluate each vertex at 2 points\r\n\t\t\theight.b *= 0.5;\r\n\r\n\t\t\t//now need to find the filtered Y and X\r\n\t\t\tfloat3 velocity_filter = VelocityRT.SampleLevel(VelocityRTSampler, curUV + veloffsets[k], 0).rgb;\r\n\r\n\t\t\tfloat2 edgevel = float2(velocity.x, velocity_filter.y);\r\n\r\n\t\t\tif (k == 1)\r\n\t\t\t{\r\n\t\t\t\tedgevel = float2(velocity_filter.x, velocity.y);\r\n\t\t\t}\r\n\r\n\t\t\tfloat2 position = (edgevel * deltaT);\r\n\t\t\r\n\t\t\tfloat2 uvpos = curUV + (position / TotalSize);\r\n\t\t\tuvpos += offsets[k];\r\n\r\n\t\t\tfloat2 celldist = abs(inUV - uvpos);\r\n\r\n\t\t\tif (celldist.x < TexelSize.x &&&& celldist.y < TexelSize.y &&&& height.b > 0.0001)\r\n\t\t\t{\r\n\r\n\t\t\t\tfloat weight = 1 - saturate(celldist.x / TexelSize.x);\r\n\t\t\t\tweight *= 1 - saturate(celldist.y / TexelSize.y);\r\n\r\n\t\t\t\t//accumulate Volume, Sediment and Height\r\n\t\t\t\taccumvol += height.b * weight;\r\n\t\t\t\taccumsediment += height.a * height.b * weight;\r\n\t\t\t\t//accumZ += (height.g+ (edgevel.z * deltaT)) * height.b * weight;\r\n\r\n\t\t\t\taccum += height.b * weight;\r\n\t\t\t}\r\n\r\n\t\t}\r\n\t}\r\n}\r\n\r\nif(accum >0)\r\n{\r\n\taccumZ /= accum;\r\n\taccumsediment /= accum;\r\n}\r\nelse\r\n{\r\n\taccumZ=startheight;\r\n\taccumvol=0;\r\n}\r\n\r\nfloat4 output = float4(startheight, startheight, accumvol, accumsediment);\r\n\r\nreturn output;"
   description=Erosion Pass
   outputType=CMOT_Float4
   inputs [
    {
     inputName=HeightRT
     input {expression=&self:MaterialExpressionTextureObjectParameter_0}
    }
    {
     inputName=VelocityRT
     input {expression=&self:MaterialExpressionTextureObjectParameter_2}
    }
    {
     inputName=deltaT
     input {expression=&self:MaterialExpressionCollectionParameter_22}
    }
    {inputName=TextureSize, input {expression=&self:MaterialExpressionAppendVector_7}}
    {inputName=TexelSize, input {expression=&self:MaterialExpressionDivide_3}}
    {inputName=TotalSize, input {expression=&self:MaterialExpressionMultiply_0}}
    {
     inputName=GridSize
     input {expression=&self:MaterialExpressionCollectionParameter_21}
    }
    {
     inputName=inUV
     input {expression=&self:MaterialExpressionMaterialFunctionCall_4}
    }
   ]
  }
  MaterialExpressionTextureObjectParameter_6 = TextureObjectParameter() {parameterName=HeightRT, texture=&Black}
  MaterialExpressionTextureObjectParameter_7 = TextureObjectParameter() {parameterName=VelocityRT, texture=&Black}
  MaterialExpressionCollectionParameter_29 = CollectionParameter() {collection=&MPC_Erosion, parameterName=deltaT}
  MaterialExpressionAppendVector_10 = AppendVector(
   A=CollectionParameter() {collection=&MPC_Erosion, parameterName=ResX}
   B=CollectionParameter() {collection=&MPC_Erosion, parameterName=ResY}
  )
  MaterialExpressionDivide_6 = Divide(B=@MaterialExpressionAppendVector_10)
  MaterialExpressionCollectionParameter_28 = CollectionParameter() {collection=&MPC_Erosion, parameterName=GridSize}
  MaterialExpressionMultiply_18 = Multiply(A=@MaterialExpressionAppendVector_10, B=@MaterialExpressionCollectionParameter_28)
  MaterialExpressionTextureCoordinate_2 = TextureCoordinate()
  n5 = Custom(
   HeightRT=@MaterialExpressionTextureObjectParameter_6
   VelocityRT=@MaterialExpressionTextureObjectParameter_7
   deltaT=@MaterialExpressionCollectionParameter_29
   TextureSize=@MaterialExpressionAppendVector_10
   TexelSize=@MaterialExpressionDivide_6
   TotalSize=@MaterialExpressionMultiply_18
   GridSize=@MaterialExpressionCollectionParameter_28
   inUV=@MaterialExpressionTextureCoordinate_2
  ) {
   code="//PositionRT: input Texture Object containing \r\n//VelcotiyRT: input \r\n//deltaT: input\r\n//input TextureSize\r\n//input TexelSize\r\n//input OutputVelocity\r\n\r\nfloat startheight = HeightRT.SampleLevel(HeightRTSampler, inUV, 0).r;\r\n\r\nint2 startidx = floor(inUV * TextureSize);\r\n\r\nfloat accumvol = 0;\r\nfloat accumZ = 0;\r\nfloat accumsediment = 0;\r\n\r\nfloat accum = 0;\r\nfloat neighbors = 0;\r\n\r\nconst int Radius = 1;\r\n\r\nfor (int j = -Radius; j <= Radius; j++)\r\n{\r\n\tfor (int i = -Radius; i <= Radius; i++)\r\n\t{\r\n\t\t\r\n\t\tfloat2 curUV = inUV + (float2(i,j) * TexelSize);\r\n\t\t//R = height, G = water height, B = water volume, A = sediment fraction\r\n\t\tfloat4 height = HeightRT.SampleLevel(HeightRTSampler, curUV, 0).rgba;\r\n\r\n\t\tfloat3 velocity = VelocityRT.SampleLevel(VelocityRTSampler, curUV, 0).rgb;\r\n\r\n\t\tfloat2 velocity_X = VelocityRT.SampleLevel(VelocityRTSampler, curUV - float2(TexelSize.x,0), 0).rg;\r\n\t\tfloat2 velocity_Y = VelocityRT.SampleLevel(VelocityRTSampler, curUV - float2(0,TexelSize.y), 0).rg;\r\n\r\n\t\tvelocity.x = ( velocity.x + velocity_X.x ) / 2;\r\n\t\tvelocity.y = ( velocity.y + velocity_Y.y ) / 2;\r\n\r\n\t\tfloat3 position = float3(0,0, height.g) + (velocity * deltaT);\r\n\t\r\n\t\tfloat2 uvpos = curUV + (position.rg / TotalSize);\r\n\r\n\t\tfloat2 celldist = abs(inUV - uvpos);\r\n\r\n\t\tif (celldist.x < TexelSize.x &&&& celldist.y < TexelSize.y &&&& height.b > 0.0001)\r\n\t\t{\r\n\r\n\t\t\tfloat weight = 1 - saturate(celldist.x / TexelSize.x);\r\n\t\t\tweight *= 1 - saturate(celldist.y / TexelSize.y);\r\n\r\n\t\t\t//accumulate Volume, Sediment and Height\r\n\t\t\taccumvol += height.b * weight;\r\n\t\t\taccumsediment += height.a * height.b * weight;\r\n\t\t\taccumZ += (height.g+ (velocity.z * deltaT)) * height.b * weight;\r\n\r\n\t\t\taccum += height.b * weight;\r\n\t\t}\r\n\t}\r\n}\r\n\r\nif(accum >0)\r\n{\r\n\taccumZ /= accum;\r\n\taccumsediment /= accum;\r\n}\r\nelse\r\n{\r\n\taccumZ=startheight;\r\n\taccumvol=0;\r\n}\r\n\r\nfloat4 output = float4(startheight, max(startheight,startheight), accumvol, accumsediment);\r\n\r\nreturn output;"
   description=Erosion Pass
   outputType=CMOT_Float4
   inputs [
    {
     inputName=HeightRT
     input {expression=&self:MaterialExpressionTextureObjectParameter_6}
    }
    {
     inputName=VelocityRT
     input {expression=&self:MaterialExpressionTextureObjectParameter_7}
    }
    {
     inputName=deltaT
     input {expression=&self:MaterialExpressionCollectionParameter_29}
    }
    {
     inputName=TextureSize
     input {expression=&self:MaterialExpressionAppendVector_10}
    }
    {inputName=TexelSize, input {expression=&self:MaterialExpressionDivide_6}}
    {inputName=TotalSize, input {expression=&self:MaterialExpressionMultiply_18}}
    {
     inputName=GridSize
     input {expression=&self:MaterialExpressionCollectionParameter_28}
    }
    {inputName=inUV, input {expression=&self:MaterialExpressionTextureCoordinate_2}}
   ]
  }
  MaterialExpressionTextureObjectParameter_3 = TextureObjectParameter() {parameterName=HeightRT, texture=&Black}
  MaterialExpressionTextureObjectParameter_4 = TextureObjectParameter() {parameterName=VelocityRT, texture=&Black}
  MaterialExpressionCollectionParameter_26 = CollectionParameter() {collection=&MPC_Erosion, parameterName=deltaT}
  MaterialExpressionAppendVector_8 = AppendVector(
   A=CollectionParameter() {collection=&MPC_Erosion, parameterName=ResX}
   B=CollectionParameter() {collection=&MPC_Erosion, parameterName=ResY}
  )
  MaterialExpressionDivide_4 = Divide(B=@MaterialExpressionAppendVector_8)
  MaterialExpressionCollectionParameter_25 = CollectionParameter() {collection=&MPC_Erosion, parameterName=GridSize}
  MaterialExpressionMultiply_4 = Multiply(A=@MaterialExpressionAppendVector_8, B=@MaterialExpressionCollectionParameter_25)
  MaterialExpressionTextureCoordinate_6 = TextureCoordinate()
  n6 = Custom(
   HeightRT=@MaterialExpressionTextureObjectParameter_3
   VelocityRT=@MaterialExpressionTextureObjectParameter_4
   deltaT=@MaterialExpressionCollectionParameter_26
   TextureSize=@MaterialExpressionAppendVector_8
   TexelSize=@MaterialExpressionDivide_4
   TotalSize=@MaterialExpressionMultiply_4
   GridSize=@MaterialExpressionCollectionParameter_25
   inUV=@MaterialExpressionTextureCoordinate_6
  ) {
   code="//PositionRT: input Texture Object containing \r\n//VelcotiyRT: input \r\n//deltaT: input\r\n//input TextureSize\r\n//input TexelSize\r\n//input OutputVelocity\r\n\r\nfloat startheight = HeightRT.SampleLevel(HeightRTSampler, inUV, 0).r;\r\n\r\nint2 startidx = floor(inUV * TextureSize);\r\n\r\nfloat accumvol = 0;\r\nfloat accumZ = 0;\r\nfloat accumsediment = 0;\r\n\r\nfloat accum = 0;\r\nfloat neighbors = 0;\r\n\r\nconst int Radius = 1;\r\n\r\nfor (int j = -Radius; j <= Radius; j++)\r\n{\r\n\tfor (int i = -Radius; i <= Radius; i++)\r\n\t{\r\n\t\t\r\n\t\tfloat2 curUV = inUV + (float2(i,j) * TexelSize);\r\n\t\t//R = height, G = water height, B = water volume, A = sediment fraction\r\n\t\tfloat4 height = HeightRT.SampleLevel(HeightRTSampler, curUV, 0).rgba;\r\n\r\n\t\tfloat3 velocity = VelocityRT.SampleLevel(VelocityRTSampler, curUV, 0).rgb;\r\n\r\n\t\tfloat2 velocity_X = VelocityRT.SampleLevel(VelocityRTSampler, curUV - float2(TexelSize.x,0), 0).rg;\r\n\t\tfloat2 velocity_Y = VelocityRT.SampleLevel(VelocityRTSampler, curUV - float2(0,TexelSize.y), 0).rg;\r\n\r\n\t\t//velocity.x = ( velocity.x + velocity_X.x ) / 2;\r\n\t\t//velocity.y = ( velocity.y + velocity_Y.y ) / 2;\r\n\r\n\t\tfloat3 position = float3(0,0, height.g) + (velocity * deltaT);\r\n\t\r\n\t\tfloat2 uvpos = curUV + (position.rg / TotalSize);\r\n\r\n\t\tfloat2 celldist = abs(inUV - uvpos);\r\n\r\n\t\tif (celldist.x < TexelSize.x &&&& celldist.y < TexelSize.y &&&& height.b > 0.0001)\r\n\t\t{\r\n\r\n\t\t\tfloat weight = 1 - saturate(celldist.x / TexelSize.x);\r\n\t\t\tweight *= 1 - saturate(celldist.y / TexelSize.y);\r\n\r\n\t\t\t//accumulate Volume, Sediment and Height\r\n\t\t\taccumvol += height.b * weight;\r\n\t\t\taccumsediment += height.a * height.b * weight;\r\n\t\t\taccumZ += (height.g+ (velocity.z * deltaT)) * height.b * weight;\r\n\r\n\t\t\taccum += height.b * weight;\r\n\t\t}\r\n\t}\r\n}\r\n\r\nif(accum >0)\r\n{\r\n\taccumZ /= accum;\r\n\taccumsediment /= accum;\r\n}\r\nelse\r\n{\r\n\taccumZ=startheight;\r\n\taccumvol=0;\r\n}\r\n\r\nfloat4 output = float4(startheight, max(startheight,startheight), accumvol, accumsediment);\r\n\r\nreturn output;"
   description=Erosion Pass
   outputType=CMOT_Float4
   inputs [
    {
     inputName=HeightRT
     input {expression=&self:MaterialExpressionTextureObjectParameter_3}
    }
    {
     inputName=VelocityRT
     input {expression=&self:MaterialExpressionTextureObjectParameter_4}
    }
    {
     inputName=deltaT
     input {expression=&self:MaterialExpressionCollectionParameter_26}
    }
    {inputName=TextureSize, input {expression=&self:MaterialExpressionAppendVector_8}}
    {inputName=TexelSize, input {expression=&self:MaterialExpressionDivide_4}}
    {inputName=TotalSize, input {expression=&self:MaterialExpressionMultiply_4}}
    {
     inputName=GridSize
     input {expression=&self:MaterialExpressionCollectionParameter_25}
    }
    {inputName=inUV, input {expression=&self:MaterialExpressionTextureCoordinate_6}}
   ]
  }
  n7 = StaticBool() {desc=Use Offset Velocities, value=true}
  n8 = StaticSwitch(
   True=@n3
   False=StaticSwitch(True=@n4, False=StaticSwitch(True=@n5, False=@n6, Value=@n7), Value=StaticBool() {desc=Use 2 sub particles, value=true})
   Value=StaticBool() {desc=Use 4 sub particles}
  )
  n9 = ComponentMask(None=@n8) {a=true, desc=Sediment Fraction}
  n10 = Multiply(A=@n9)
  n11 = Divide(B=AppendVector(
    A=CollectionParameter() {collection=&MPC_Erosion, parameterName=ResX}
    B=CollectionParameter() {collection=&MPC_Erosion, parameterName=ResY}
   ))
  n12 = Subtract(
   A=MaterialFunctionCall() {materialFunction=&PointSampleUVs}.TexelCenterUVs
   B=Multiply(A=Constant2Vector() {r=1}, B=@n11)
  )
  n13 = TextureSampleParameter2D!(
   UVs=MaterialFunctionCall() {materialFunction=&PointSampleUVs}.TexelCenterUVs
   Level
   Apply View MipBias
  ) {constMipValue=0, mipValueMode=TMVM_MipLevel, parameterName=VelocityRTCurrent, texture=&Black}
  n14 = Subtract(
   A=MaterialFunctionCall() {materialFunction=&PointSampleUVs}.TexelCenterUVs
   B=Multiply(A=@n11, B=Constant2Vector() {g=1})
  )
  n15 = AppendVector(
   A=Add(
    A=TextureSampleParameter2D(UVs=@n12) {constMipValue=0, mipValueMode=TMVM_MipLevel, parameterName=VelocityRTCurrent, texture=&Black}.R
    B=@n13.R
   )
   B=Add(
    A=TextureSampleParameter2D(UVs=@n14) {constMipValue=0, mipValueMode=TMVM_MipLevel, parameterName=VelocityRTCurrent, texture=&Black}.G
    B=@n13.G
   )
  )
  n16 = MaterialFunctionCall(Vector 2=Multiply(A=Divide(A=@n15), B=CollectionParameter() {collection=&MPC_Erosion, parameterName=deltaT})) {materialFunction=/Engine/Functions/Engine_MaterialFunctions02/Utility/VectorLength}
  n17 = Multiply(
   A=Divide(A=Add(A=@n16.V2 Length), B=CollectionParameter() {collection=&MPC_Erosion, parameterName=GridSize})
   B=CollectionParameter() {collection=&MPC_Erosion, parameterName=Solubility}
  )
  n18 = Subtract(A=@n10, B=@n17) {desc=Positive if Over saturated}
  n19 = Ceil(None=@n18)
  n20 = Multiply(
   A=@n18
   B=LinearInterpolate(
    A=ScalarParameter() {defaultValue=0.1, parameterName=Per Step Dissolution}
    B=ScalarParameter() {defaultValue=0.1, parameterName=Per Step Deposition}
    Alpha=Saturate(None=@n19)
   )
  )
  n21 = ComponentMask(None=@n8) {b=true}
  n22 = Multiply(A=@n20, B=@n21) {desc=Change in disolved sediment}
  n23 = Multiply(
   A=@n21
   B=Multiply(
    A=CollectionParameter() {collection=&MPC_Erosion, parameterName=Evaporation}
    B=CollectionParameter() {collection=&MPC_Erosion, parameterName=deltaT}
   )
  )
  n24 = MaterialFunctionCall() {materialFunction=/Landmass/Landscape/OldPrototype_BP/Mats/MF/VelocityAdvect}
  n25 = MaterialFunctionCall() {materialFunction=&PointSampleUVs}
  n26 = Constant()
  n27 = MaterialFunctionCall(Vector 2=@n24.OffsetWorldSpace) {materialFunction=&VectorLength}
  n28 = CollectionParameter() {collection=&MPC_Erosion, parameterName=MaxAdvectSlope}
  n29 = Max(
   A=TextureSampleParameter2D(UVs=@n24.Advected UVs, Level=Constant()) {mipValueMode=TMVM_MipLevel, parameterName=HeightRT, texture=&Black}.R
   B=Subtract(
    A=TextureSampleParameter2D(UVs=@n25.TexelCenterUVs, Level=@n26) {mipValueMode=TMVM_MipLevel, parameterName=HeightRT, texture=&Black}.R
    B=Multiply(A=@n27.V2 Length, B=@n28) {constB=0.1}
   )
  )
  n30 = CollectionParameter() {collection=&MPC_Erosion, parameterName=RainAmount}
  n31 = CollectionParameter() {collection=&MPC_Erosion, parameterName=deltaT}
  n32 = Subtract(
   A=Add(
    A=Multiply(A=@n23, B=@n9)
    B=MaterialFunctionCall!(UVs=@n24.Advected UVs, Height=@n29) {materialFunction=/Landmass/Landscape/OldPrototype_BP/Mats/MF/SlopeSmoothing}.Height
   )
   B=Multiply(A=@n17, B=Multiply(A=@n30, B=@n31))
  )
  n33 = MaterialFunctionCall() {materialFunction=&PointSampleUVs}
  n34 = Constant()
  n35 = CollectionParameter() {collection=&MPC_Erosion, parameterName=Evaporation}
  n36 = CollectionParameter() {collection=&MPC_Erosion, parameterName=deltaT}
  n37 = Subtract(A=@n21, B=Multiply(A=@n21, B=Multiply(A=@n35, B=@n36)))
  n38 = CollectionParameter() {collection=&MPC_Erosion, parameterName=RainAmount}
  n39 = CollectionParameter() {collection=&MPC_Erosion, parameterName=deltaT}
  AppendVector(
   A=AppendVector(
    A=Add(A=@n22, B=@n32)
    B=TextureSampleParameter2D(UVs=@n33.TexelCenterUVs, Level=@n34) {mipValueMode=TMVM_MipLevel, parameterName=HeightRT, texture=&Black}.G
   )
   B=Add(A=Max(A=@n37) {constB=0}, B=Multiply(A=@n38, B=@n39))
  ) {desc="B = Sim Results"}
  OneMinus(None=Saturate(None=Subtract(A=@n10, B=@n20)))
  Saturate(None=Ceil(None=TextureSampleParameter2D!(UVs, Apply View MipBias) {parameterName=PositionRT, texture=&Black}.A))
  ScalarParameter() {parameterName=RainAmount}
  ScalarParameter() {parameterName=deltaT}
  ScalarParameter() {defaultValue=0.5, parameterName=so}
  TextureCoordinate()
  TextureCoordinate()
  TextureSampleParameter2D() {desc="A= Current Sediment Fraction", parameterName=HeightRT, texture=&Black}
 ]
}
