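# Unlit prototype material for the landscape erosion brush that computes water velocity by gather-style advection from HeightRT and VelocityRT.
# Custom HLSL moves neighboring texels' velocities by deltaT, then averages those landing nearby, weighted by water volume and bilinear distance.
# Static switches choose among three variants; defaults select one sample per texel, despite the "Use 2 sub particles" label, over the staggered-grid two- and four-sample versions.
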Material /Landmass/Landscape/OldPrototype_BP/Mats/ScatterGather/GatherVel
graph {
 properties {
  bAllowNegativeEmissiveColor=true
  bUsedWithStaticLighting=true
  pixelDepthOffsetMode=PDOM_Legacy
  shadingModel=MSM_Unlit
  usedShadingModels=Unlit
 }
 nodes [
  n1 = Multiply(
   A=TextureCoordinate()
   B=AppendVector(A=ScalarParameter() {defaultValue=2048, parameterName=SizeX}, B=ScalarParameter() {defaultValue=2048, parameterName=SizeY})
  )
  n2 = VectorNoise(World Position=AppendVector(A=Floor(None=@n1), B=Constant()))
  AppendVector(
   A=Multiply(A=ComponentMask(None=@n2) {g=true, r=true}) {constB=4}
   B=Add(
    A=ScalarParameter() {defaultValue=25, parameterName=Z}
    B=TextureSampleParameter2D() {parameterName=HeightRT, texture=&Black}.R
   )
  )
  Saturate(None=Ceil(None=TextureSampleParameter2D() {parameterName=PositionRT, texture=&Black}.A))
  ScalarParameter() {parameterName=OutputVelocity_1}
  ScalarParameter() {parameterName=OutputVelocity_2}
  MaterialExpressionTextureObjectParameter_1 = TextureObjectParameter() {parameterName=HeightRT, texture=&Black}
  MaterialExpressionTextureObjectParameter_5 = TextureObjectParameter() {parameterName=VelocityRT, texture=&Black}
  MaterialExpressionCollectionParameter_18 = CollectionParameter() {collection=/Landmass/Landscape/BlueprintBrushes/MPC/MPC_Erosion, parameterName=deltaT}
  MaterialExpressionAppendVector_6 = AppendVector(
   A=CollectionParameter() {collection=&MPC_Erosion, parameterName=ResX}
   B=CollectionParameter() {collection=&MPC_Erosion, parameterName=ResY}
  )
  MaterialExpressionDivide_0 = Divide(B=@MaterialExpressionAppendVector_6)
  MaterialExpressionCollectionParameter_2 = CollectionParameter() {collection=&MPC_Erosion, parameterName=GridSize}
  MaterialExpressionMultiply_14 = Multiply(A=@MaterialExpressionAppendVector_6, B=@MaterialExpressionCollectionParameter_2)
  MaterialExpressionMaterialFunctionCall_1 = MaterialFunctionCall() {materialFunction=/Landmass/Landscape/OldPrototype_BP/Mats/MF/PointSampleUVs}
  n3 = Custom!(
   HeightRT=@MaterialExpressionTextureObjectParameter_1
   VelocityRT=@MaterialExpressionTextureObjectParameter_5
   deltaT=@MaterialExpressionCollectionParameter_18
   TextureSize=@MaterialExpressionAppendVector_6
   TexelSize=@MaterialExpressionDivide_0
   TotalSize=@MaterialExpressionMultiply_14
   GridSize=@MaterialExpressionCollectionParameter_2
   inUV=@MaterialExpressionMaterialFunctionCall_1.TexelCenterUVs
  ) {
   code="//PositionRT: input Texture Object containing \r\n//VelcotiyRT: input \r\n//deltaT: input\r\n//input TextureSize\r\n//input TexelSize\r\n//input OutputVelocity\r\n\r\nfloat startheight = HeightRT.SampleLevel(HeightRTSampler, inUV, 0).r;\r\n\r\nint2 startidx = floor(inUV * TextureSize);\r\n\r\nfloat3 accumvel = 0;\r\nfloat2 accum = 0;\r\n\r\nconst int Radius = 1;\r\n\r\nconst float2 posoffsets[4] = {float2(TexelSize.x,0), float2(-TexelSize.x,0), float2(0,TexelSize.y), float2(0,-TexelSize.y)};\r\n\r\nfor (int j = -Radius; j <= Radius; j++)\r\n{\r\n\tfor (int i = -Radius; i <= Radius; i++)\r\n\t{\r\n\t\t\r\n\t\tfloat2 curUV = inUV + (float2(i,j) * TexelSize);\r\n\t\t//R = height, G = water height, B = water volume, A = sediment fraction\r\n\t\tfloat4 height = HeightRT.SampleLevel(HeightRTSampler, curUV, 0).rgba;\r\n\t\theight.b /= 4;\r\n\r\n\t\tfloat2 centervelocity = VelocityRT.SampleLevel(VelocityRTSampler, curUV, 0).rg;\r\n\t\tfloat2 offsetXvelocity = VelocityRT.SampleLevel(VelocityRTSampler, curUV - float2(TexelSize.x,0), 0).rg;\r\n\t\tfloat2 offsetYvelocity = VelocityRT.SampleLevel(VelocityRTSampler, curUV - float2(0,TexelSize.y), 0).rg;\r\n\t\t\r\n\t\tfloat2 filter1velocity = VelocityRT.SampleLevel(VelocityRTSampler, curUV + float2(+TexelSize.x * 0.5, -TexelSize.y *0.5), 0).rg;\r\n\t\tfloat2 filter2velocity = VelocityRT.SampleLevel(VelocityRTSampler, curUV + float2(-TexelSize.x * 0.5, -TexelSize.y *0.5), 0).rg;\r\n\t\tfloat2 filter3velocity = VelocityRT.SampleLevel(VelocityRTSampler, curUV + float2(-TexelSize.x * 0.5, +TexelSize.y *0.5), 0).rg;\r\n\t\t\r\n\t\tfloat2 velocities[4];\r\n\t\t\r\n\t\tvelocities[0] = float2( centervelocity.x, filter1velocity.y);\r\n\t\tvelocities[1] = float2( offsetXvelocity.x, filter2velocity.y);\r\n\t\tvelocities[2] = float2( filter3velocity.x, centervelocity.y);\r\n\t\tvelocities[3] = float2( filter2velocity.x, offsetYvelocity.y);\r\n\t\t\r\n\t\tfor (int k = 0; k < 4; k++)\r\n\t\t{\r\n\t\t\tfloat2 position = velocities[k] * deltaT;\r\n\t\t\r\n\t\t\tfloat2 uvpos = curUV + (position.rg / TotalSize);\r\n\t\t\t//uvpos += posoffsets[k] * 0.5;\r\n\t\r\n\t\t\tfloat2 walldistX = abs(inUV + float2(0.5*TexelSize.x,0)- uvpos);\r\n\t\t\tfloat2 walldistY = abs(inUV + float2(0,0.5*TexelSize.y)- uvpos);\r\n\t\r\n\t\t\tif (walldistX.x < TexelSize.x &&&& walldistX.y < TexelSize.y &&&& height.b > 0.0001)\r\n\t\t\t{\r\n\t\t\t\tfloat weight = 1 - saturate(walldistX.x / TexelSize.x);\r\n\t\t\t\tweight *= 1 - saturate(walldistX.y / TexelSize.y);\r\n\t\r\n\t\t\t\t//Volume weighted sums\r\n\t\t\t\taccumvel.x += velocities[k].x * height.b * weight;\r\n\t\t\t\taccum.x += height.b * weight;\r\n\t\t\t}\r\n\t\t\t\r\n\t\t\tif (walldistY.x < TexelSize.x &&&& walldistY.y < TexelSize.y &&&& height.b > 0.0001)\r\n\t\t\t{\r\n\t\t\t\tfloat weight = 1 - saturate(walldistY.x / TexelSize.x);\r\n\t\t\t\tweight *= 1 - saturate(walldistY.y / TexelSize.y);\r\n\t\r\n\t\t\t\t//Volume weighted sums\r\n\t\t\t\taccumvel.y += velocities[k].y * height.b * weight;\r\n\t\t\t\taccum.y += height.b * weight;\r\n\t\t\t}\r\n\t\t\t\r\n\t\t}\r\n\t}\r\n}\r\n\r\n//Divide by weighted sum\r\n\r\nif(accum.x > 0)\r\n{\r\n\taccumvel.x /= accum.x;\r\n}\r\nif(accum.y > 0)\r\n{\r\n\taccumvel.y /= accum.y;\r\n}\r\n\r\nreturn accumvel;"
   description=Erosion Pass
   inputs [
    {
     inputName=HeightRT
     input {expression=&self:MaterialExpressionTextureObjectParameter_1}
    }
    {
     inputName=VelocityRT
     input {expression=&self:MaterialExpressionTextureObjectParameter_5}
    }
    {
     inputName=deltaT
     input {expression=&self:MaterialExpressionCollectionParameter_18}
    }
    {inputName=TextureSize, input {expression=&self:MaterialExpressionAppendVector_6}}
    {inputName=TexelSize, input {expression=&self:MaterialExpressionDivide_0}}
    {inputName=TotalSize, input {expression=&self:MaterialExpressionMultiply_14}}
    {
     inputName=GridSize
     input {expression=&self:MaterialExpressionCollectionParameter_2}
    }
    {inputName=inUV, input {expression=&self:MaterialExpressionMaterialFunctionCall_1}}
   ]
  }
  MaterialExpressionTextureObjectParameter_0 = TextureObjectParameter() {parameterName=HeightRT, texture=&Black}
  MaterialExpressionTextureObjectParameter_2 = TextureObjectParameter() {parameterName=VelocityRT, texture=&Black}
  MaterialExpressionCollectionParameter_6 = CollectionParameter() {collection=&MPC_Erosion, parameterName=deltaT}
  MaterialExpressionAppendVector_1 = AppendVector(
   A=CollectionParameter() {collection=&MPC_Erosion, parameterName=ResX}
   B=CollectionParameter() {collection=&MPC_Erosion, parameterName=ResY}
  )
  MaterialExpressionDivide_1 = Divide(B=@MaterialExpressionAppendVector_1)
  MaterialExpressionCollectionParameter_5 = CollectionParameter() {collection=&MPC_Erosion, parameterName=GridSize}
  MaterialExpressionMultiply_0 = Multiply(A=@MaterialExpressionAppendVector_1, B=@MaterialExpressionCollectionParameter_5)
  MaterialExpressionMaterialFunctionCall_0 = MaterialFunctionCall() {materialFunction=&PointSampleUVs}
  n4 = Custom(
   HeightRT=@MaterialExpressionTextureObjectParameter_0
   VelocityRT=@MaterialExpressionTextureObjectParameter_2
   deltaT=@MaterialExpressionCollectionParameter_6
   TextureSize=@MaterialExpressionAppendVector_1
   TexelSize=@MaterialExpressionDivide_1
   TotalSize=@MaterialExpressionMultiply_0
   GridSize=@MaterialExpressionCollectionParameter_5
   inUV=@MaterialExpressionMaterialFunctionCall_0.TexelCenterUVs
  ) {
   code="//PositionRT: input Texture Object containing \r\n//VelcotiyRT: input \r\n//deltaT: input\r\n//input TextureSize\r\n//input TexelSize\r\n//input OutputVelocity\r\n\r\n\r\nint2 startidx = floor(inUV * TextureSize);\r\n\r\nfloat3 accumvel = 0;\r\nfloat accum = 0;\r\nfloat neighbors = 0;\r\n\r\nconst int Radius = 2;\r\n\r\nfor (int j = -Radius; j <= Radius; j++)\r\n{\r\n\tfor (int i = -Radius; i <= Radius; i++)\r\n\t{\r\n\t\t\r\n\t\tfloat2 curUV = inUV + (float2(i,j) * TexelSize);\r\n\t\t//R = height, G = water height, B = water volume, A = sediment fraction\r\n\t\tfloat4 height = HeightRT.SampleLevel(HeightRTSampler, curUV, 0).rgba;\r\n\t\tfloat3 velocity = VelocityRT.SampleLevel(VelocityRTSampler, curUV, 0).rgb;\r\n\r\n\t\tfloat3 position = (velocity * deltaT);\r\n\t\r\n\t\tfloat2 uvpos = curUV + (position.rg / TotalSize);\r\n\r\n\t\tfloat2 celldist = abs(inUV - uvpos);\r\n\r\n\t\tif (celldist.x < TexelSize.x &&&& celldist.y < TexelSize.y &&&& height.b > 0.0001)\r\n\t\t{\r\n\r\n\t\t\tfloat weight = 1 - saturate(celldist.x / TexelSize.x);\r\n\t\t\tweight *= 1 - saturate(celldist.y / TexelSize.y);\r\n\r\n\t\t\t//Volume weighted sums\r\n\t\t\taccumvel += velocity * height.b * weight;\r\n\t\t\taccum += height.b * weight;\r\n\t\t}\r\n\t}\r\n}\r\n\r\n//Divide by weighted sum\r\n\r\nif(accum > 0)\r\n{\r\n\taccumvel /= accum;\r\n}\r\n\r\nreturn accumvel;"
   description=Erosion Pass
   inputs [
    {
     inputName=HeightRT
     input {expression=&self:MaterialExpressionTextureObjectParameter_0}
    }
    {
     inputName=VelocityRT
     input {expression=&self:MaterialExpressionTextureObjectParameter_2}
    }
    {
     inputName=deltaT
     input {expression=&self:MaterialExpressionCollectionParameter_6}
    }
    {inputName=TextureSize, input {expression=&self:MaterialExpressionAppendVector_1}}
    {inputName=TexelSize, input {expression=&self:MaterialExpressionDivide_1}}
    {inputName=TotalSize, input {expression=&self:MaterialExpressionMultiply_0}}
    {
     inputName=GridSize
     input {expression=&self:MaterialExpressionCollectionParameter_5}
    }
    {inputName=inUV, input {expression=&self:MaterialExpressionMaterialFunctionCall_0}}
   ]
  }
  MaterialExpressionTextureObjectParameter_3 = TextureObjectParameter() {parameterName=HeightRT, texture=&Black}
  MaterialExpressionTextureObjectParameter_4 = TextureObjectParameter() {parameterName=VelocityRT, texture=&Black}
  MaterialExpressionCollectionParameter_10 = CollectionParameter() {collection=&MPC_Erosion, parameterName=deltaT}
  MaterialExpressionAppendVector_4 = AppendVector(
   A=CollectionParameter() {collection=&MPC_Erosion, parameterName=ResX}
   B=CollectionParameter() {collection=&MPC_Erosion, parameterName=ResY}
  )
  MaterialExpressionDivide_2 = Divide(B=@MaterialExpressionAppendVector_4)
  MaterialExpressionCollectionParameter_9 = CollectionParameter() {collection=&MPC_Erosion, parameterName=GridSize}
  MaterialExpressionMultiply_2 = Multiply(A=@MaterialExpressionAppendVector_4, B=@MaterialExpressionCollectionParameter_9)
  MaterialExpressionMaterialFunctionCall_2 = MaterialFunctionCall() {materialFunction=&PointSampleUVs}
  n5 = Custom(
   HeightRT=@MaterialExpressionTextureObjectParameter_3
   VelocityRT=@MaterialExpressionTextureObjectParameter_4
   deltaT=@MaterialExpressionCollectionParameter_10
   TextureSize=@MaterialExpressionAppendVector_4
   TexelSize=@MaterialExpressionDivide_2
   TotalSize=@MaterialExpressionMultiply_2
   GridSize=@MaterialExpressionCollectionParameter_9
   inUV=@MaterialExpressionMaterialFunctionCall_2.TexelCenterUVs
  ) {
   code="//PositionRT: input Texture Object containing \r\n//VelcotiyRT: input \r\n//deltaT: input\r\n//input TextureSize\r\n//input TexelSize\r\n//input OutputVelocity\r\n\r\nint2 startidx = floor(inUV * TextureSize);\r\n\r\nfloat3 accumvel = 0;\r\nfloat accum = 0;\r\n\r\nconst float2 offsets[2] = {float2(0.5 * TexelSize.x,0), float2(0,0.5 * TexelSize.y)};\r\nconst float2 veloffsets[2] = {float2(0.5 * TexelSize.x,-0.5 * TexelSize.y), float2(-0.5 * TexelSize.x,0.5 * TexelSize.y)};\r\n\r\nconst int Radius =3;\r\n\r\nfor (int j = -Radius; j < Radius; j++)\r\n{\r\n\tfor (int i = -Radius; i < Radius; i++)\r\n\t{\r\n\t\t\r\n\t\tfloat2 curUV = inUV + (float2(i,j) * TexelSize);\r\n\r\n\t\t//R = height, G = water height, B = water volume, A = sediment fraction\r\n\t\tfloat4 height = HeightRT.SampleLevel(HeightRTSampler, curUV, 0).rgba;\r\n\r\n\t\t//Offset grid velocity\r\n\t\tfloat3 velocity = VelocityRT.SampleLevel(VelocityRTSampler, curUV, 0).rgb;\r\n\r\n\t\tfor (int k = 0; k < 2; k++)\r\n\t\t{\r\n\r\n\t\t\t//Get water at each of 2 boundaries\r\n\t\t\tfloat4 height = HeightRT.SampleLevel(HeightRTSampler, curUV + offsets[k], 0).rgba;\r\n\t\t\t\r\n\t\t\t//half since we evaluate each vertex at 2 points\r\n\t\t\theight.b *= 0.5;\r\n\r\n\t\t\t//now need to find the filtered Y and X\r\n\t\t\tfloat3 velocity_filter = VelocityRT.SampleLevel(VelocityRTSampler, curUV + veloffsets[k], 0).rgb;\r\n\r\n\t\t\tfloat2 edgevel = float2(velocity.x, velocity_filter.y);\r\n\r\n\t\t\tif (k == 1)\r\n\t\t\t{\r\n\t\t\t\tedgevel = float2(velocity_filter.x, velocity.y);\r\n\t\t\t}\r\n\r\n\t\t\tfloat2 position = (edgevel * deltaT);\r\n\t\t\r\n\t\t\tfloat2 uvpos = curUV + (position / TotalSize);\r\n\t\t\tuvpos += offsets[k];\r\n\r\n\t\t\tfloat2 celldistX = abs(inUV + offsets[k] - uvpos);\r\n\t\t\tfloat2 celldistY = abs(inUV + offsets[k] - uvpos);\r\n\r\n\t\t\tif (celldistX.x < TexelSize.x &&&& celldistX.y < TexelSize.y &&&& height.b > 0.0001)\r\n\t\t\t{\r\n\r\n\t\t\t\tfloat weight = 1 - saturate(celldistX.x / TexelSize.x);\r\n\t\t\t\tweight *= 1 - saturate(celldistX.y / TexelSize.y);\r\n\r\n\t\t\t\t//Volume weighted sums\r\n\t\t\t\taccumvel.x += edgevel.x * height.b * weight;\r\n\t\t\t\taccum += height.b * weight;\r\n\t\t\t}\r\n\t\t\t\r\n\t\t\tif (celldistY.x < TexelSize.x &&&& celldistY.y < TexelSize.y &&&& height.b > 0.0001)\r\n\t\t\t{\r\n\r\n\t\t\t\tfloat weight = 1 - saturate(celldistY.x / TexelSize.x);\r\n\t\t\t\tweight *= 1 - saturate(celldistY.y / TexelSize.y);\r\n\r\n\t\t\t\t//Volume weighted sums\r\n\t\t\t\taccumvel.y += edgevel.y * height.b * weight;\r\n\t\t\t\taccum += height.b * weight;\r\n\t\t\t}\r\n\r\n\t\t}\r\n\t}\r\n}\r\n\r\n//Divide by weighted sum\r\n\r\nif(accum > 0)\r\n{\r\n\taccumvel /= accum;\r\n}\r\n\r\nreturn accumvel;"
   description=Erosion Pass
   inputs [
    {
     inputName=HeightRT
     input {expression=&self:MaterialExpressionTextureObjectParameter_3}
    }
    {
     inputName=VelocityRT
     input {expression=&self:MaterialExpressionTextureObjectParameter_4}
    }
    {
     inputName=deltaT
     input {expression=&self:MaterialExpressionCollectionParameter_10}
    }
    {inputName=TextureSize, input {expression=&self:MaterialExpressionAppendVector_4}}
    {inputName=TexelSize, input {expression=&self:MaterialExpressionDivide_2}}
    {inputName=TotalSize, input {expression=&self:MaterialExpressionMultiply_2}}
    {
     inputName=GridSize
     input {expression=&self:MaterialExpressionCollectionParameter_9}
    }
    {inputName=inUV, input {expression=&self:MaterialExpressionMaterialFunctionCall_2}}
   ]
  }
  StaticSwitch(
   True=ComponentMask(None=@n3) {b=true, g=true, r=true}
   False=StaticSwitch(
    True=ComponentMask(None=@n4) {b=true, g=true, r=true}
    False=ComponentMask(None=@n5) {b=true, g=true, r=true}
    Value=StaticBool() {desc=Use 2 sub particles, value=true}
   )
   Value=StaticBool() {desc=Use 4 sub particles}
  )
  TextureCoordinate()
  TextureCoordinate()
  TextureCoordinate()
 ]
}
